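# Sound cue for Clyde's Hoagie vehicle horn music, playing a looping 141 BPM music track.
# A single looping wave player is the root, with no mixing or randomization.
# It uses the Hoagie horn distant attenuation and horn music concurrency, a music POI sound class, a boosted volume and a long audible range.
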SoundCue /Clyde/Valet/Hoagie/HornMusicCues/Clyde_Hoagie_Horn_Music_Full_Cue
root=WavePlayer_0
nodes [{id=WavePlayer_0, properties {bLooping=true, soundWaveAssetPtr=&/MX_HoagieVehicle_Clyde_Loop_BPM141}}]
properties {
 attenuationSettings=/Hoagie/Audio/Hoagie_Horn_Distant_Att
 maxDistance=23500
 soundClassObject=/Game/Sounds/FortSoundClassesAndModes/Fort_Music_POI_PSM
 volumeMultiplier=1.5
 concurrencySet [/Hoagie/Audio/Assets/Horn/Hoagie_Horn_Music_Conc]
}
