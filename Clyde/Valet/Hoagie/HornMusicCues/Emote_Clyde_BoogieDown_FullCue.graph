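# SoundCue that plays the Boogie Down emote music as a horn music cue, using the Hoagie distant horn attenuation.
# A Mixer layers a one-shot intro with the looping main track, each behind its own fixed Delay so the loop enters half a second after the intro starts.
# It uses the first-person emote music sound class and concurrency, sends to the music audio-analysis submix, and has a PreventNew rule in its concurrency overrides.
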SoundCue /Clyde/Valet/Hoagie/HornMusicCues/Emote_Clyde_BoogieDown_FullCue
root=Mixer_0
nodes [
 {id=Delay_0, children [WavePlayer_0], properties {delayMax=0.7, delayMin=0.7}}
 {id=Delay_1, children [WavePlayer_1], properties {delayMax=0.2, delayMin=0.2}}
 {id=Mixer_0, children [Delay_0,Delay_1], properties {inputVolume [1,1]}}
 {id=WavePlayer_0, properties {bLooping=true, soundWaveAssetPtr=/Game/Athena/Sounds/Emotes/emote_boogiedown_A}}
 {id=WavePlayer_1, properties {soundWaveAssetPtr=/Game/Athena/Sounds/Emotes/emote_boogiedown_intro_A}}
]
properties {
 attenuationSettings=/Hoagie/Audio/Hoagie_Horn_Distant_Att
 maxDistance=23500
 soundClassObject=/Game/Sounds/FortSoundClassesAndModes/Fort_Music_Emote_1P_PSM
 volumeMultiplier=1
 concurrencyOverrides {resolutionRule=PreventNew}
 concurrencySet [/Game/Sounds/Concurrency_Rules/Athena_EmoteMusic_1P_Conc]
 soundSubmixSends [{minSendLevel=0, soundSubmix=&Fortnite_AudioAnalysis_Music}]
}
